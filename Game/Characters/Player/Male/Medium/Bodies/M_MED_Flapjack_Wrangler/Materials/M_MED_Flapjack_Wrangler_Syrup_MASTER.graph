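# Master character material for the syrup on the Flapjack Wrangler skin, with a timed animated drip effect and optional Slurp skin FX.
# A curve atlas row, timed from a Blueprint-supplied game time, drives wobble, stretch and normal strength.
# Vertex offset math combines a height gradient, sine waves and downward stretch relative to a head or hat socket.
# Panning world-aligned fold normals on pre-skinned position blend in between shared character PreFX and PostFX functions.
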Material /Game/Characters/Player/Male/Medium/Bodies/M_MED_Flapjack_Wrangler/Materials/M_MED_Flapjack_Wrangler_Syrup_MASTER
graph {
 properties {
  bUseMaterialAttributes=true
  bUsedWithMeshDeformer=true
  bUsedWithMorphTargets=true
  bUsedWithSkeletalMesh=true
  pixelDepthOffsetMode=PDOM_Legacy
  shadingModel=MSM_FromMaterialExpression
  usedShadingModels=Default Lit | Subsurface Profile
 }
 nodes [
  n1 = MaterialFunctionCall() {materialFunction=&BaseFN_CharacterFunction_PreFX}
  n2 = Divide(
   A=@n1.Pre skinned Local Position
   B=MaterialFunctionCall() {materialFunction=&ObjectLocalBounds}.Local Bounds Max
  )
  n3 = Subtract(A=OneMinus(None=ComponentMask(None=@n2) {b=true}), B=ScalarParameter() {defaultValue=-0.01, parameterName=GradientOffset})
  n4 = ScalarParameter() {defaultValue=3, parameterName=FracPart}
  n5 = Power(Base=Frac(None=Multiply(A=@n3, B=@n4)))
  n6 = VectorParameter() {parameterName=Drip_WPO_Amt, defaultValue=<rgba 0 0 -48 1>}
  n7 = Subtract(A=Time(), B=ScalarParameter() {defaultValue=100000000, group=TargetPractice, parameterName=FJ_GameTimeFromBP})
  n8 = ScalarParameter() {defaultValue=4, group=TargetPractice, parameterName=FG_ReactiveGlowDuration}
  n9 = CurveAtlasRowParameter(CurveTime=Saturate(None=Divide(A=@n7, B=@n8))) {
   atlas=/Game/Characters/Shared/CurveAtlas/AthenaCharacterCurveAtlas
   curve=&/SyrupDripCurve
   defaultValue=9
   parameterName=DripCurve
  }
  n10 = VertexColor()
  n11 = ConstantBiasScale(None=Sine(None=Multiply(A=@n10.B) {constB=2}))
  n12 = Multiply(A=@n1.Pre skinned Local Position, B=ScalarParameter() {defaultValue=0.1, parameterName=TimeOffsetStripes}) {constB=0.05, desc=1}
  n13 = Power(Base=ConstantBiasScale(None=Sine(None=@n12)))
  n14 = VertexNormalWS()
  n15 = LinearInterpolate(A=ComponentMask(None=@n13) {g=true}, B=ComponentMask(None=@n13) {r=true}, Alpha=Abs(None=ComponentMask(None=@n14) {g=true}))
  n16 = Multiply(A=Add(A=Time(), B=Add(A=@n11, B=@n15)), B=ScalarParameter() {defaultValue=0.35, parameterName=DripSpeed})
  n17 = Multiply(A=Multiply(A=@n5, B=Multiply(A=@n6.RGB, B=@n9.R)), B=ConstantBiasScale(None=Sine(None=@n16)))
  n18 = WorldPosition()
  n19 = LinearInterpolate(
   A=VectorParameter() {parameterName=headSocketLocation, defaultValue=<rgba 0 0 0 1>}.RGB
   B=VectorParameter() {parameterName=hatSocketLocation, defaultValue=<rgba 0 0 0 1>}.RGB
   Alpha=ScalarParameter() {parameterName=IsSyrupHat}
  )
  n20 = Subtract(A=@n18.XYZ, B=@n19)
  n21 = Add(A=Multiply(A=Normalize(VectorInput=@n20)) {constB=0}, B=VectorParameter() {parameterName=WPO_Offset, defaultValue=<rgba 0 0 0 1>}.RGB)
  n22 = Constant2Vector()
  n23 = VectorParameter() {parameterName=Pivot_Offset, defaultValue=<rgba 0 0 0 1>}
  n24 = Normalize(VectorInput=Subtract(A=Add(A=@n19, B=@n23.RGB), B=@n18.XYZ))
  n25 = Add(A=Multiply(A=Multiply(A=@n15) {constB=0.25}, B=@n9.G), B=@n9.G)
  n26 = ScalarParameter() {defaultValue=-6, parameterName=Drip_StretchAmt}
  n27 = Clamp(None=Multiply(A=ComponentMask(None=@n24) {b=true}, B=Multiply(A=@n25, B=@n26) {constB=-4})) {maxDefault=0, minDefault=-10000000}
  Add(B=Add(A=Add(A=@n17, B=@n21), B=AppendVector(A=@n22, B=@n27))) {constB=0}
  Add(A=@n16)
  Add(B=Time())
  n28 = TextureSample(Tex=TextureObjectParameter() {group=Skin FX, parameterName=SkinFX_Mask})
  n29 = StaticComponentMaskParameter(None=@n28.RGBA) {defaultR=true, group=Skin FX, parameterName=Slurp Channel}
  n30 = StaticComponentMaskParameter(None=@n28.RGBA) {defaultG=true, group=Skin FX, parameterName=Slurp Bubbles Channel}
  n31 = StaticComponentMaskParameter(None=@n28.RGBA) {defaultB=true, group=Skin FX, parameterName=Slurp Normal Distortion Channel}
  n32 = GetMaterialAttributes(None=StaticSwitchParameter(
    True=MaterialFunctionCall!(Alpha=@n29, AO=@n1.AO, BubblesMask=@n30, SlurpNormalDistortionMask=@n31, Pre Skinned Position=@n1.AO, Material=@n1.AO) {materialFunction=/Game/Weapons/WeaponSkins/MaterialFunction/MF_Slurp}.Result
    False=@n1.Result
   ) {group=Skin FX, parameterName=UseSlurp}) {attributeGetTypes [#1]}
  n33 = TextureObject() {samplerType=SAMPLERTYPE_Normal, texture=/Game/Gadgets/Assets/VinderTech_GliderChute/Glider_BlackMondayCape/Textures/T_Folds_N}
  n34 = VectorParameter() {parameterName=SyrupDrip_Tiling, defaultValue=<rgba 24 24 24 1>}
  n35 = VectorParameter() {group=BlackMonday, parameterName=FoldsPanSpeed, defaultValue=<rgba 0.1 0.1 -0.2 1>}
  n36 = ComponentMask(None=@n35.RGB) {b=true, g=true}
  n37 = ComponentMask(None=@n35.RGB) {b=true, r=true}
  n38 = ComponentMask(None=@n35.RGB) {g=true, r=true}
  n39 = MaterialFunctionCall(
   Normal=ComponentMask(None=MaterialFunctionCall!(
     TextureObject=@n33
     TextureSize=@n34.RGB
     Normal
     WorldSpace
     ProjectionTransitionContrast
     WorldPosition=@n1.Pre skinned Local Position
     Use High Quality Normals
     Pan_YZ=@n36
     Pan_XZ=@n37
     Pan_XY=@n38
    ) {materialFunction=/Game/Weapons/WeaponSkins/Wraps/MaterialFunctions/WorldAlignedNormal_ComplexPan}.XYZ Texture) {b=true, g=true, r=true}
   Flatness=LinearInterpolate(B=ScalarParameter() {parameterName=DripNormalIntensity}, Alpha=@n9.B) {constA=1}
  ) {materialFunction=&FlattenNormal}
  MaterialFunctionCall(
   AO=@n1.AO
   Distance Blend=@n1.AO
   Fresnel=@n1.AO
   Gmap=@n1.AO
   Pre Skinned Local Position=@n1.AO
   Material=SetMaterialAttributes!(
    MaterialAttributes=@n32.MaterialAttributes
    Normal=MaterialFunctionCall(BaseNormal=@n32.Normal, AdditionalNormal=@n39.Result) {materialFunction=&BlendAngleCorrectedNormals}.Result
   ) {attributeSetTypes [#1]}
  ) {materialFunction=&BaseFN_CharacterFunction_PostFX}
  MaterialFunctionCall() {materialFunction=&ObjectPivotPoint}
  PreSkinnedPosition()
  Reroute(None=@n5)
  ScalarParameter() {defaultValue=1, parameterName=Syrup_DripControl}
  ScalarParameter() {defaultValue=1, parameterName=DripControl}
 ]
}
